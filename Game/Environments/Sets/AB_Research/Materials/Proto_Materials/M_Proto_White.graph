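# M_Proto_White is a prototype Default Lit material in the AB_Research environment set, named as a plain white.
# It holds just two constant nodes, a light grey color and a mid-range scalar, with no textures or parameters.
# The graph does not show which material inputs those constants feed.
# It is flagged for use with Nanite meshes.
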Material /Game/Environments/Sets/AB_Research/Materials/Proto_Materials/M_Proto_White
graph {
 properties {bUsedWithNanite=true, pixelDepthOffsetMode=PDOM_Legacy, usedShadingModels=Default Lit}
 nodes [
  Constant3Vector() {constant=<rgba 0.625 0.625 0.625 0>}
  Constant() {r=0.5}
 ]
}
